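# Translucent unlit particle material for the Kaleidoscope aura, drawing an animated, circular kaleidoscope pattern on Niagara sprites and meshes.
# Radial UVs get a sine-stepped angular repeat and tiled distance, then Frac and Abs mirror them, with time offset by a per-particle random value.
# These UVs sample pattern and digital-noise textures under a soft circle mask, with a dynamic parameter blending in dots.
# Panning cell noise and digital noise drive hue shifts of the particle color.
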Material /SparksCosmetics/Items/Auras/Assets/NS_Aura_Kaleido/M_Aura_Kaleidoscope
graph {
 properties {
  bAutomaticallySetUsageInEditor=false
  bUsedWithMeshParticles=true
  bUsedWithNiagaraSprites=true
  blendMode=BLEND_Translucent
  shadingModel=MSM_Unlit
  usedShadingModels=Unlit
 }
 nodes [
  n1 = Add(A=Time(), B=DynamicParameter() {paramNames [Random Offset, Use Dots, Param3, Param4]}.Random Offset)
  n2 = Sine(None=Multiply(A=ScalarParameter() {defaultValue=1, parameterName=Speed Scale}, B=@n1)) {period=75}
  n3 = Constant() {r=4}
  n4 = Constant() {r=6}
  n5 = Truncate!(None=Divide(A=MaterialFunctionCall!(Sine Phase=@n2, Value 1=@n3, Value 2=@n4) {materialFunction=/Engine/Functions/Engine_MaterialFunctions02/Utility/Sine_Remapped}.Remapped Sine)) {is=class MaterialExpressionTruncate}
  n6 = TextureCoordinate()
  n7 = MaterialFunctionCall(Vector or UVs=Multiply(A=Subtract(A=@n6) {constB=0.5}, B=ScalarParameter() {defaultValue=0.25, parameterName=Pattern Scale}) {constB=0.25}) {materialFunction=&VectorToRadialValue}
  n8 = AppendVector(
   A=Add(A=Multiply(A=@n5, B=@n7.Vector Converted to Angle), B=@n2)
   B=Multiply(A=@n7.Linear Distance, B=ScalarParameter() {defaultValue=5, parameterName=Stacked Tile})
  )
  n9 = Abs(None=ConstantBiasScale(None=Frac(None=@n8)) {bias=-0.5, scale=2})
  n10 = Add(
   A=@n9
   B=MaterialFunctionCall(Sine Phase=@n2, Value 1=Constant(), Value 2=Constant() {r=2}) {materialFunction=&Sine_Remapped}.Remapped Sine
  )
  n11 = MaterialFunctionCall!(Hardness=Constant() {r=0.5}, Center, Radius=Constant() {r=0.5}) {materialFunction=/Game/Packages/Fortress_SharedMaterials/Base_Material_Functions/FX_SphereMask2D}
  n12 = Multiply(
   A=Add(
    A=TextureSample(UVs=@n10) {texture=&/T_Aura_Kaleidoscope}.R
    B=Multiply(A=@n11.Result) {constB=0.9}
   )
   B=@n11.Result
  )
  n13 = TextureSample(UVs=@n10) {texture=&T_Aura_Kaleidoscope}
  LinearInterpolate(A=@n12, B=Multiply(A=@n13.B, B=@n12), Alpha=DynamicParameter() {paramNames [Random Offset, Use Dots, Param3, Param4]}.Use Dots)
  n14 = ParticleColor()
  n15 = TextureSample(UVs=Panner(Coordinate=@n9, Time=@n1) {speedX=0.1, speedY=-0.1}) {desc="Hue Shift Texture ", texture=/Game/Global/Textures/Default/Noise/T_SoftCells_Negative}
  n16 = LinearInterpolate(Alpha=@n11.Result) {constA=5, constB=0.75}
  n17 = ScalarParameter() {defaultValue=0.25, parameterName=hue shift perecentage}
  n18 = TextureSample(UVs=@n10) {texture=/Game/Effects/Fort_Effects/Effects/Atlas/Textures/T_Atlas_DigitalNoise_02}
  n19 = Add(A=@n18.R, B=@n18.R)
  n20 = MaterialFunctionCall(Hue Shift Percentage=Multiply(A=Power(Base=@n15.R, Exp=@n16), B=Multiply(A=@n17, B=@n19)) {constB=0.2}, Texture=@n14.RGB) {materialFunction=&HueShift}
  MaterialFunctionCall(Emissive=LinearInterpolate(A=@n14.RGB, B=@n20.Result, Alpha=@n13.B)) {materialFunction=&MF_FX_EmissiveAdjustments}
  Multiply(A=@n2, B=ScalarParameter() {defaultValue=0.1, parameterName=Rotation Scale})
  Multiply(A=@n10) {constB=0.5}
  n21 = Constant() {r=0.1}
  n22 = Constant() {r=0.2}
  Multiply(
   A=@n20.Result
   B=Multiply(
    A=@n18.R
    B=MaterialFunctionCall(Input=@n12, Input Low=@n21, Input High=@n22) {materialFunction=&RemapValue_0-1}.Result
   )
  )
  ScalarParameter() {defaultValue=4, parameterName=sides}
  ScalarParameter() {defaultValue=4, parameterName=offset}
 ]
}
