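# Translucent unlit Niagara sprite material that draws animated kaleidoscope lines with a soft pulsing glow for the Kaleido cosmetic aura.
# Rotated polar UVs are repeated by a side count and mirror-folded to sample scrolling swirl noise inside a circular mask.
# That pattern blends two hue shifts of particle color, one varied by a panning polar kaleidoscope texture.
# Opacity adds sine-driven radial rings to the alpha-scaled pattern, and a dynamic parameter offsets pattern timing.
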Material /SparksCosmetics/Items/Auras/Assets/NS_Aura_Kaleido/M_Aura_Kaleidoscope_Lines
graph {
 properties {
  bAutomaticallySetUsageInEditor=false
  bUsedWithNiagaraSprites=true
  blendMode=BLEND_Translucent
  shadingModel=MSM_Unlit
  usedShadingModels=Unlit
 }
 nodes [
  n1 = MaterialFunctionCall!(Hardness=Constant() {r=0.5}, Center, Radius=Constant() {r=0.5}) {materialFunction=/Game/Packages/Fortress_SharedMaterials/Base_Material_Functions/FX_SphereMask2D}
  n2 = Multiply(A=Add(A=Multiply(A=@n1.Result) {constB=2}) {constB=-1}, B=ScalarParameter() {defaultValue=0.1, parameterName=Glow Hue Shift}) {constB=0.2}
  n3 = ParticleColor()
  n4 = MaterialFunctionCall(Vector or UVs=Add(A=TextureCoordinate()) {constB=-0.5}) {materialFunction=&VectorToRadialValue}
  n5 = Multiply(A=@n4.Linear Distance) {constB=0.5}
  n6 = TextureSample(UVs=Panner(Coordinate=AppendVector(A=@n4.Vector Converted to Angle, B=@n5)) {speedY=-0.05}) {texture=&/T_Aura_Kaleidoscope}
  n7 = MaterialFunctionCall(Hardness=Constant() {r=0.5}, Radius=Constant() {r=0.45}) {materialFunction=&FX_SphereMask2D}
  n8 = Multiply(
   A=Power(Base=Multiply(A=@n6.B, B=@n7.Result), Exp=LinearInterpolate(Alpha=@n7.Result) {constA=1.5, constB=0.5})
   B=ScalarParameter() {defaultValue=0.25, parameterName=Pattern Hue Shift}
  ) {constB=0.2}
  n9 = ScalarParameter() {defaultValue=1, parameterName=Rotation Speed Scale}
  n10 = Multiply(
   A=Add(A=Time(), B=DynamicParameter() {paramNames [Random Offset, Use Dots, Param3, Param4]}.Random Offset)
   B=ScalarParameter() {defaultValue=-0.3, parameterName=Speed Scale}
  )
  n11 = ScalarParameter() {defaultValue=3, parameterName=Number of Sides}
  n12 = Rotator(Time=@n10) {speed=-1}
  n13 = MaterialFunctionCall(Vector or UVs=Multiply(A=Add(A=@n12) {constB=-0.5}, B=ScalarParameter() {defaultValue=3, parameterName=Pattern Scale}) {constB=0.25}) {materialFunction=&VectorToRadialValue}
  n14 = AppendVector(A=Add(A=Multiply(A=@n9, B=@n10), B=Multiply(A=@n11, B=@n13.Vector Converted to Angle)), B=@n13.Linear Distance)
  n15 = Abs(None=ConstantBiasScale(None=Frac(None=@n14)) {bias=-0.5, scale=2})
  n16 = Multiply(
   A=@n7.Result
   B=TextureSample(UVs=Add(A=@n15, B=@n10)) {texture=/BRCosmetics/Accessories/FORT_Backpacks/Backpack_M_MED_EthicalNoggin/Textures/T_EthicalNoggin_Backpack_SwirlNoise}.G
  )
  n17 = Constant()
  n18 = Constant() {r=0.1}
  MaterialFunctionCall(Emissive=LinearInterpolate(
    A=MaterialFunctionCall(Hue Shift Percentage=@n2, Texture=@n3.RGB) {materialFunction=&HueShift}.Result
    B=MaterialFunctionCall(Hue Shift Percentage=@n8, Texture=@n3.RGB) {materialFunction=&HueShift}.Result
    Alpha=MaterialFunctionCall(Input=@n16, Input Low=@n17, Input High=@n18) {materialFunction=&RemapValue_0-1}.Result
   )) {materialFunction=&MF_FX_EmissiveAdjustments}
  n19 = Add(A=TextureCoordinate()) {constB=-0.5}
  n20 = Multiply(A=@n19, B=@n19)
  n21 = Add(A=ComponentMask(None=@n20) {r=true}, B=ComponentMask(None=@n20) {g=true})
  n22 = Add(
   A=Multiply(A=SquareRoot(None=@n21), B=ScalarParameter() {defaultValue=-6, parameterName=Glow Tiling}) {constB=-6}
   B=Multiply(A=Time(), B=ScalarParameter() {defaultValue=2, parameterName=Glow Speed}) {constB=2}
  )
  n23 = Add(A=Multiply(A=Sine(None=@n22)) {constB=0.5}) {constB=0.5}
  n24 = ScalarParameter() {defaultValue=0.1, parameterName=Glow Opacity}
  NamedRerouteDeclaration(None=Multiply(A=Multiply(A=@n23, B=@n24) {constB=0.1}, B=@n1.Result)) {name=GlowOpacity, nodeColor=<rgba 0 1 0.3906 1>}
  Saturate(None=Add(A=Multiply(A=@n3.A, B=@n16), B=NamedRerouteUsage()))
 ]
}
