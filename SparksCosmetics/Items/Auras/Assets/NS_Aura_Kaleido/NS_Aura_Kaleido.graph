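# Kaleido cosmetic aura: a Niagara system of color-shifting sprites around a character, scaled for large bodies and dimmed in the frontend.
# A bright backer, kaleidoscope sprite and eleven oscillating, time-of-day-scaled rays follow a shared system center, while GPU circles drift outward with curl noise.
# Ten decal/glow sprites sit on or between paired skeleton bones and cycle through glow colors.
# A shared system color curve and timing tint the kaleidoscope, rays and circles.
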NiagaraSystem /SparksCosmetics/Items/Auras/Assets/NS_Aura_Kaleido/NS_Aura_Kaleido
dependencies {
 usedDataInterfaces [
  &NiagaraDataInterfaceColorCurve
  &NiagaraDataInterfaceArrayFloat
  &NiagaraDataInterfaceArrayInt32
  &NiagaraDataInterfaceCurve
  &NiagaraDataInterfaceArrayFloat2
  &NiagaraDataInterfaceSkeletalMesh
  &NiagaraDataInterfaceArrayColor
 ]
 usedDynamicInputs [&MakeVector]
 usedModules [
  &self:SystemSpawnScript.SystemScriptSource.SystemScriptGraph.NiagaraNodeAssignment_0.SetVariables_#1
  &self:SystemSpawnScript.SystemScriptSource.SystemScriptGraph.NiagaraNodeAssignment_3.SetVariables_#2
  &self:SystemSpawnScript.SystemScriptSource.SystemScriptGraph.NiagaraNodeAssignment_1.SetVariables_#3
  &SystemState
  &SampleSkeletalMesh
  &self:SystemSpawnScript.SystemScriptSource.SystemScriptGraph.NiagaraNodeAssignment_2.SetVariables_#4
  &self:SystemSpawnScript.SystemScriptSource.SystemScriptGraph.NiagaraNodeAssignment_5.SetVariables_#5
 ]
}
emitters {
 BrightBacker {
  emitterUpdateScript [
   &EmitterState()
   &SpawnBurst_Instantaneous(Use Spawn Probability=false, UseLoopCountLimit=true)
  ]
  particleSpawnScript [&InitializeParticle(
    Color=<rgba 3 0.1732 0 0.7>
    Lifetime=3
    Sprite Size Mode=::NewEnumerator3
    Uniform Sprite Size=&Multiply_Float(A=135, B=&MakeCustomFloatFromBool(Bool=$User.IsLargeBody, True Float=1.4, False Float=1))
    UsePositionOffset=false
    Write Material Random=true
   ) {module=InitializeParticle001}]
  particleUpdateScript [
   &ParticleState(
    Kill Particles When Lifetime Has Elapsed=false
    Let Infinitely Lived Particles Die When Emitter Deactivates=true
    Loop Particles Lifetime=true
   )
   &SpriteRotationRate(Rotation Rate=5)
   &CameraOffset(Camera Offset Amount=-20)
   &ScaleColor(
    Scale Alpha=&Multiply_Float(A=1, B=&MakeCustomFloatFromBool(Bool=$User.IsFrontEnd, True Float=0.6, False Float=1))
    ScaleA=true
    ScaleRGB=true
   )
   &DynamicMaterialParameters(
    Index 0 Param 1=2
    Index 0 Param 2=1
    Index 0 Param 4=1
    Module.Index0Float0Write=true
    Module.Index0Float1Write=true
    Module.Index0Float2Write=true
    Module.Index0Float3Write=true
   )
   /SetVariables(Particles.Position=$System.CenteredPos) {isSetParameters=true}
  ]
  rendererClasses [&NiagaraSpriteRendererProperties]
  renderers [{rendererClass=&NiagaraSpriteRendererProperties}]
 }
 CirclesOut {
  simTarget=GPUComputeSim
  emitterUpdateScript [
   &EmitterState()
   &SpawnRate(
    Spawn Probability=&MakeCustomFloatFromBool(Bool=$User.IsFrontEnd, True Float=0.5, False Float=1)
    SpawnRate=100
    Use Spawn Probability=true
   )
  ]
  particleSpawnScript [
   &InitializeParticle(
    Color=&ColorFromCurve(
     ColorCurve=$System.ColorShift
     CurveIndex=&Frac_Float(Float=&Add_Float(A=$System.ColorTiming, B=&RandomRangeFloat(Minimum=-0.2, Maximum=0.2)))
    )
    Lifetime Max=1
    Lifetime Min=0.2
    Lifetime Mode=::NewEnumerator1
    Position Offset:NiagaraPosition=$System.CenteredPos
    Position Offset Coordinate Space=::Local
    Sprite Size Mode=::NewEnumerator1
    Uniform Sprite Size Max=15
    UsePositionOffset=true
    Write Material Random=true
   )
   &AddVelocity(
    Rotation Coordinate Space=::Simulation
    Velocity=&RandomVector(Vector Scale=&Multiply_Float(A=100, B=&MakeCustomFloatFromBool(Bool=$User.IsLargeBody, True Float=1.4, False Float=1.2)))
   )
   /SetVariables(Particles.Velocity=&Multiply_Vector(A=$Particles.Velocity, B=<1.25 1 1.25>)) {enabled=false, isSetParameters=true}
  ]
  particleUpdateScript [
   &ParticleState()
   &ScaleSpriteSize(
    Uniform Curve Scale=&MakeCustomFloatFromBool(Bool=$User.IsLargeBody, True Float=1.4, False Float=1)
    Uniform Curve Sprite Scale=DataInterface() {propertyValues {Curve {defaultValue=0, keys [{tangentMode=RCTM_User, time=0, value=1}, {tangentMode=RCTM_User, time=1, value=0}]}}}
   )
   &Drag(Drag=2, Use Linear Drag=true, Use Rotational Drag=true)
   &CurlNoiseForce(Bias Noise=false, Noise Strength=100, Pan Noise=false)
   &SolveForcesAndVelocity(Clamp Velocity=false, Limit Acceleration=false)
   /SetVariables(
    Particles.BackerColor:LinearColor=<rgba 0.0176 0.0128 0.0885 0.5>
    Particles.BackerPivotOffset:Vector2f=&SelectVector2DFromArray(
     Array Sampling Mode=::NewEnumerator0
     Direct Array Index=?
     Vector 2D Selection Array=DataInterface() {propertyValues {FloatData [<0.4 0.4>, <0.4 0.6>, <0.6 0.4>, <0.6 0.6>]}}
     Random Seed=0
    )
    Particles.BackerScale:Vector2f=&Multiply_Vector2D_ByFloat(Float=1.75, Vector2D=$Particles.SpriteSize)
   ) {isSetParameters=true}
   &CameraOffset(Camera Offset Amount=-39)
   &ScaleColor(
    Scale RGB=&VectorFromFloat(Value=&MakeCustomFloatFromBool(Bool=$User.IsFrontEnd, True Float=0.5, False Float=1))
    ScaleA=true
    ScaleRGB=true
   )
  ]
  rendererClasses [&NiagaraSpriteRendererProperties]
  renderers [
   {rendererClass=&NiagaraSpriteRendererProperties}
   {rendererClass=&NiagaraSpriteRendererProperties, rendererIndex=1}
  ]
 }
 Decals+Glows {
  emitterUpdateScript [
   &EmitterState()
   &SpawnBurst_Instantaneous(Spawn Count=10, Use Spawn Probability=false, UseLoopCountLimit=true)
  ]
  particleSpawnScript [
   &InitializeParticle(
    Color=<rgba 0.0394 0.0287 0.1983 25>
    Lifetime=3
    Sprite Rotation Mode=::NewEnumerator0
    Sprite Size Mode=::NewEnumerator3
    Uniform Sprite Size=&Multiply_Float(
     A=&SelectFloatFromArray(
      Direct Array Index=&ReturnExecIndex()
      Float Selection Array=DataInterface() {propertyValues {FloatData [25,25,20,20,20,20,17,15,17,15]}}
     )
     B=&MakeCustomFloatFromBool(Bool=$User.IsLargeBody, True Float=1.5, False Float=1)
    )
    UsePositionOffset=false
    Write Material Random=true
   ) {module=InitializeParticle001}
   &CameraOffset(Camera Offset Amount=&SelectFloatFromArray(
     Direct Array Index=&ReturnExecIndex()
     Float Selection Array=DataInterface() {propertyValues {FloatData [10,10,8,5,8,5]}}
    ))
   /SetVariables(Particles.glowScale:Vector2f=&Vector2DFromFloat(Value=&Multiply_Float(
      A=&SelectFloatFromArray(
       Direct Array Index=&ReturnExecIndex()
       Float Selection Array=DataInterface() {propertyValues {FloatData [35,35,28,28,28,28,25,22,25,22]}}
      )
      B=&MakeCustomFloatFromBool(Bool=$User.IsLargeBody, True Float=1.4, False Float=1)
     ))) {isSetParameters=true}
   /SetVariables(Particles.PickBone:NiagaraInt32=&SelectIntFromArray(
     Direct Array Index=&ReturnExecIndex()
     Int Selection Array=DataInterface() {propertyValues {IntData [1,2,3,3,4,4,5,5,6,6]}}
    )) {isSetParameters=true}
   /SetVariables(Particles.BoneCompanion:NiagaraInt32=&SelectIntFromArray(Direct Array Index=$Particles.PickBone, Int Selection Array=$System.CompanionBone)) {isSetParameters=true}
  ]
  particleUpdateScript [
   &ParticleState(
    Kill Particles When Lifetime Has Elapsed=false
    Let Infinitely Lived Particles Die When Emitter Deactivates=true
    Loop Particles Lifetime=true
   )
   &SkeletalMeshLocation(
    Bone / Socket Index=$Particles.PickBone
    Bone Sampling Mode=::NewEnumerator4
    MaskSpawnGroup=false
    Orientation Sampling=::NewEnumerator1
    Position Sampling=::NewEnumerator2
    Sampled Position Offset Coordinate Space:ENiagaraSkelMeshTransforms=::NewEnumerator2
    Skeletal Mesh=$System.Skeletal Mesh
   )
   &SkeletalMeshLocation(
    Bone / Socket Index=$Particles.BoneCompanion
    Bone Sampling Mode=::NewEnumerator4
    MaskSpawnGroup=false
    Orientation Sampling=::NewEnumerator1
    Position Sampling=::NewEnumerator2
    Skeletal Mesh=$System.Skeletal Mesh
   ) {module=SkeletalMeshLocation001}
   /SetVariables(Particles.Position=&LerpPosition(
     A=$Particles.SkeletalMeshLocation.SampledPosition
     B=$Particles.SkeletalMeshLocation001.SampledPosition
     Alpha=&SelectFloatFromArray(
      Direct Array Index=&ReturnExecIndex()
      Float Selection Array=DataInterface() {propertyValues {FloatData [0,0,0.5,0.1,0.5,0.1,0.6,0.25,0.6,0.25]}}
     )
    )) {isSetParameters=true}
   &DynamicMaterialParameters(
    Index 0 Param 1=&SelectFloatFromArray(
     Direct Array Index=&ReturnExecIndex()
     Float Selection Array=DataInterface() {propertyValues {FloatData [5,5,3,3,3,3,4,4,4,4]}}
    )
    Index 0 Param 3=1
    Index 1 Param 1=2
    Index 1 Param 2=1
    Module.Param0WriteEnabled=true
    Module.Param0WriteEnabled1=true
    Module.Param1WriteEnabled=true
    Module.Param1WriteEnabled1=true
    Module.Param2WriteEnabled=false
    Module.Param2WriteEnabled1=true
    Module.Param3WriteEnabled=false
    Module.Param3WriteEnabled1=true
    Write Parameter Index 1=true
   )
   &Color(Color=&ColorFromCurve(
     ColorCurve=DataInterface() {propertyValues {
       AlphaCurve {keys [<at 0 1 RCIM_Cubic>]}
       BlueCurve {keys [<at 0 0 RCIM_Cubic>,<at 0.5 0>,<at 1 1 RCIM_Cubic>]}
       GreenCurve {keys [<at 0 0 RCIM_Cubic>,<at 0.5 1>,<at 1 0 RCIM_Cubic>]}
       RedCurve {keys [<at 0 1 RCIM_Cubic>,<at 0.5 0>,<at 1 0 RCIM_Cubic>]}
      }}
     CurveIndex=&ReturnNormalizedExecIndex()
    )) {enabled=false}
   &InterpolateOverTime(
    Rate Of Change=3
    Target Color Value:LinearColor=&SelectLinearColorFromArray(
     Array Sampling Mode=::NewEnumerator1
     Direct Array Index=&Add_Integer(A=&ReturnExecIndex(), B=&MakeIntFromFloat(Float=$Engine.Time))
     Direct Array Mode=::NewEnumerator1
     Color Selection Array=$System.Glow Colors
     Random Seed=?
    )
    Variable Type=::NewEnumerator4
   )
  ]
  rendererClasses [&NiagaraSpriteRendererProperties]
  renderers [
   {rendererClass=&NiagaraSpriteRendererProperties}
   {rendererClass=&NiagaraSpriteRendererProperties, rendererIndex=1}
  ]
 }
 Kaleidoscope {
  emitterSpawnScript [/SetVariables(Emitter.GlowOpacity:NiagaraFloat=&MakeCustomFloatFromBool(Bool=$User.IsFrontEnd, True Float=0.05, False Float=0.1)) {isSetParameters=true}]
  emitterUpdateScript [
   &EmitterState()
   &SpawnBurst_Instantaneous(Use Spawn Probability=false, UseLoopCountLimit=true)
  ]
  particleSpawnScript [&InitializeParticle(
    Color=<rgba 2 2 2 1>
    Lifetime=5
    Sprite Size Mode=::NewEnumerator3
    Uniform Sprite Size=&Multiply_Float(A=100, B=&MakeCustomFloatFromBool(Bool=$User.IsLargeBody, True Float=1.4, False Float=1.2))
    UsePositionOffset=false
    Write Material Random=true
   ) {module=InitializeParticle001}]
  particleUpdateScript [
   &ParticleState(Kill Particles When Lifetime Has Elapsed=false, Let Infinitely Lived Particles Die When Emitter Deactivates=true)
   /SetVariables(
    Particles.Linear Color=<rgba 0.0176 0.0128 0.0885 5>
    Particles.Vector 2D:Vector2f=&Multiply_Vector2D_ByFloat(Float=1.05, Vector2D=$Particles.SpriteSize)
   ) {isSetParameters=true}
   &CameraOffset(Camera Offset Amount=-20) {enabled=false}
   /SetVariables(Particles.CameraOffset=-40, Particles.cameraoffset2=-41) {isSetParameters=true}
   &ScaleColor(
    Scale Alpha=&Multiply_Float(A=1, B=&MakeCustomFloatFromBool(Bool=$User.IsFrontEnd, True Float=0.6, False Float=1))
    ScaleA=true
    ScaleRGB=true
   ) {enabled=false}
   &DynamicMaterialParameters(
    Index 0 Param 1=&RandomRangeFloat(Minimum=0, Maximum=1)
    Index 0 Param 2=0.5
    Module.Index0Float0Write=true
    Module.Index0Float1Write=true
    Module.Index0Float2Write=true
    Module.Index0Float3Write=true
   )
   &ScaleColor(
    Scale Mode=::NewEnumerator1
    Scale RGBA=&MakeVector4_FromLinearColor(LinearColor=&ColorFromCurve(ColorCurve=$System.ColorShift, CurveIndex=&Frac_Float(Float=$System.ColorTiming)))
    ScaleRGBA=true
   ) {module=ScaleColor001}
   /SetVariables(Particles.Position=$System.CenteredPos) {isSetParameters=true}
  ]
  rendererClasses [&NiagaraSpriteRendererProperties]
  renderers [
   {rendererClass=&NiagaraSpriteRendererProperties}
   {rendererClass=&NiagaraSpriteRendererProperties, rendererIndex=1}
  ]
 }
 RaysStatic {
  emitterUpdateScript [
   &EmitterState()
   &SpawnBurst_Instantaneous(Spawn Count=11, Use Spawn Probability=false, UseLoopCountLimit=true)
  ]
  particleSpawnScript [
   &InitializeParticle(
    Color=&ColorFromCurve(ColorCurve=$System.ColorShift, CurveIndex=&ReturnNormalizedExecIndex(Normalized Index Includes One=false))
    Lifetime=7
    Sprite Rotation Angle=&Multiply_FloatByInt(Float=&Divide_Float(A=360, B=11), Integer=&ReturnExecIndex())
    Sprite Rotation Mode=::NewEnumerator0
    Sprite Size=&Multiply_Vector2D(
     A=&RandomRangeVector2D(Minimum=<30 70>, Maximum=<30 71>)
     B=&MakeCustomVector2DFromBool(Bool=$User.IsLargeBody, True Vector2D=<1.2 1.2>, False Vector2D=<1 1>)
    )
    Sprite Size Mode=::NewEnumerator4
    UsePositionOffset=false
    Write Material Random=true
   )
   &CameraOffset(Camera Offset Amount=-50) {module=CameraOffset001}
  ]
  particleUpdateScript [
   &ParticleState(Kill Particles When Lifetime Has Elapsed=false)
   &SpriteRotationRate(Rotation Rate=&Multiply_Float(
     A=50
     B=&Sine(
      Normalized Angle=&Add_Float(A=$Particles.NormalizedAge, B=&RandomRangeFloat(Minimum=-0.05, Maximum=0.05))
      Period=2
      Scale=1
      Bias=0
     )
    ))
   &ScaleSpriteSize(
    Scale Factor=&MakeVector2D(
     X=&FloatFromCurve(FloatCurve=DataInterface() {propertyValues {Curve {keys [
          {interpMode=RCIM_Cubic, tangentMode=RCTM_User, time=0, value=1}
          {interpMode=RCIM_Cubic, tangentMode=RCTM_User, time=0.33, value=1.1}
          <at 0.66 1 RCIM_Cubic>
         ]}}})
     Y=&FloatFromCurve(
      FloatCurve=DataInterface() {propertyValues {Curve {keys [
          {interpMode=RCIM_Cubic, tangentMode=RCTM_User, time=0, value=0.75}
          {interpMode=RCIM_Cubic, tangentMode=RCTM_User, time=0.25, value=1.1}
          {tangentMode=RCTM_User, time=0.5, value=0.75}
         ]}}}
      CurveIndex=&Frac_Float(Float=&Add_Float(A=$Particles.NormalizedAge, B=&RandomRangeFloat(Minimum=0, Maximum=1)))
     )
    )
    Scale Sprite Size Mode=::NewEnumerator2
   )
   /SetVariables(Particles.Position=$System.CenteredPos) {isSetParameters=true}
   /SetVariables(Particles.Color=&MakeLinearColorFromVectorAndFloat(
     "Vector (RGB)"=&Multiply_Vector(
      A=&MakeVectorFromLinearColorRGB(Color=&LerpLinearColors(
        StartColor=$Particles.Initial.Color
        EndColor=&ColorFromCurve(ColorCurve=$System.ColorShift, CurveIndex=&Frac_Float(Float=$System.ColorTiming))
        LerpFactor=0.25
       ))
      B=&VectorFromFloat(Value=&FNTimeOfDayScalar(
        Night Scale=0.9
        Morning Scale=0.9
        Day Scale=&MakeCustomFloatFromBool(Bool=$User.IsFrontEnd, True Float=0.3, False Float=1)
        Evening Scale=0.9
       ))
     )
     "Float (Alpha)"=&FNTimeOfDayScalar(Night Scale=0.35, Morning Scale=0.45, Day Scale=0.9, Evening Scale=0.45)
    )) {isSetParameters=true}
  ]
  rendererClasses [&NiagaraSpriteRendererProperties]
  renderers [{rendererClass=&NiagaraSpriteRendererProperties}]
 }
}
summary {
 systemName=NS_Aura_Kaleido
 emitters [
  {emitterName=BrightBacker, rendererClasses [&NiagaraSpriteRendererProperties]}
  {emitterName=Kaleidoscope, rendererClasses [&NiagaraSpriteRendererProperties]}
  {emitterName=Decals+Glows, rendererClasses [&NiagaraSpriteRendererProperties]}
  {emitterName=RaysStatic, rendererClasses [&NiagaraSpriteRendererProperties]}
  {emitterName=CirclesOut, simTarget=GPUComputeSim, rendererClasses [&NiagaraSpriteRendererProperties]}
 ]
 userVariables [
  User.bIsFrontEnd {type=NiagaraBool}
  User.bIsFrontEndPreview {type=NiagaraBool}
  User.FrontEndScale {type=NiagaraFloat, defaultValue=0.75, description=""}
  User.IsFrontEnd {type=NiagaraBool}
  User.IsLargeBody {type=NiagaraBool}
 ]
}
