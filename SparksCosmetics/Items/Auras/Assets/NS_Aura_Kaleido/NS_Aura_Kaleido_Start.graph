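# Niagara start burst for the Kaleido aura cosmetic, a short multicolour pop on the character.
# Flare, DarkGlow, Corona and rotating Rays sprites follow System.CenteredPos, with skeletal mesh sampling at system level and time-of-day scaling on Flare flash and Corona alpha.
# GPU CirclesOut sprites fly outward under curl noise and drag, with a dark offset backer and speed and size scaled by User.IsLargeBody.
# Rays and circles pick random colours from matching rainbow curves.
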NiagaraSystem /SparksCosmetics/Items/Auras/Assets/NS_Aura_Kaleido/NS_Aura_Kaleido_Start
dependencies {
 usedDataInterfaces [
  &NiagaraDataInterfaceColorCurve
  &NiagaraDataInterfaceCurve
  &NiagaraDataInterfaceVector2DCurve
  &NiagaraDataInterfaceArrayFloat
  &NiagaraDataInterfaceArrayFloat2
  &NiagaraDataInterfaceSkeletalMesh
 ]
 usedDynamicInputs [&SelectFloatFromArray, &LerpPosition]
 usedModules [
  &self:SystemSpawnScript.SystemScriptSource.SystemScriptGraph.NiagaraNodeAssignment_0.SetVariables_#1
  &SystemState
  &SampleSkeletalMesh
  &self:SystemSpawnScript.SystemScriptSource.SystemScriptGraph.NiagaraNodeAssignment_2.SetVariables_#2
 ]
}
emitters {
 CirclesOut {
  simTarget=GPUComputeSim
  emitterUpdateScript [
   &EmitterState()
   &SpawnBurst_Instantaneous(Spawn Count=50, Use Spawn Probability=false, UseLoopCountLimit=false)
  ]
  particleSpawnScript [
   &InitializeParticle(
    Color=&ColorFromCurve(
     ColorCurve=DataInterface() {propertyValues {
       AlphaCurve {keys [<at 0 1 RCIM_Cubic>]}
       BlueCurve {keys [
         <at 0 0.7157 RCIM_Cubic>
         {arriveTangentWeight=0.0579, interpMode=RCIM_Cubic, leaveTangentWeight=0.1039, tangentMode=RCTM_User, time=0.1568, value=2}
         {arriveTangent=-6.0034, interpMode=RCIM_Cubic, leaveTangent=-6.0034, time=0.3653, value=0.5068}
         {arriveTangent=-2.0605, interpMode=RCIM_Cubic, leaveTangent=-2.0605, time=0.4785, value=0.0689}
         {arriveTangent=-0.2853, interpMode=RCIM_Cubic, leaveTangent=-0.2853, time=0.6113, value=0}
         {arriveTangent=0.0766, interpMode=RCIM_Cubic, leaveTangent=0.0766, time=0.7199, value=0}
         {arriveTangent=2.5554, interpMode=RCIM_Cubic, leaveTangent=2.5554, time=0.8722, value=0.02}
         <at 1 0.7157 RCIM_Cubic>
        ]}
       GreenCurve {keys [
         <at 0 0.4969 RCIM_Cubic>
         {arriveTangent=-1.3602, interpMode=RCIM_Cubic, leaveTangent=-1.3602, time=0.1568, value=0.1369}
         {arriveTangent=-0.2799, interpMode=RCIM_Cubic, leaveTangent=-0.2799, time=0.3653, value=0}
         {arriveTangent=1.0968, interpMode=RCIM_Cubic, leaveTangent=1.0968, time=0.4785, value=0.0469}
         {arriveTangent=3.7429, interpMode=RCIM_Cubic, leaveTangent=3.7429, time=0.6113, value=0.2698}
         {arriveTangent=2.7988, interpMode=RCIM_Cubic, leaveTangent=2.7988, time=0.7199, value=0.9505}
         {arriveTangent=-1.6195, interpMode=RCIM_Cubic, leaveTangent=-1.6195, time=0.8722, value=1}
         <at 1 0.4969 RCIM_Cubic>
        ]}
       RedCurve {keys [
         <at 0 0 RCIM_Cubic>
         {arriveTangentWeight=0.1864, interpMode=RCIM_Cubic, leaveTangentWeight=0.3345, tangentMode=RCTM_User, time=0.1568, value=0}
         {arriveTangentWeight=0.2264, interpMode=RCIM_Cubic, leaveTangentWeight=0.1399, tangentMode=RCTM_User, time=0.3653, value=1.5}
         {arriveTangentWeight=0.1082, interpMode=RCIM_Cubic, leaveTangentWeight=0.0892, tangentMode=RCTM_User, time=0.4785, value=1}
         {arriveTangentWeight=0.0478, interpMode=RCIM_Cubic, leaveTangentWeight=0.0203, tangentMode=RCTM_User, time=0.6113, value=1}
         {arriveTangentWeight=0.175, interpMode=RCIM_Cubic, leaveTangentWeight=0.16, tangentMode=RCTM_User, time=0.7199, value=1}
         {arriveTangentWeight=0.1029, interpMode=RCIM_Cubic, leaveTangentWeight=0.236, tangentMode=RCTM_User, time=0.8722, value=0.0018}
         <at 1 0 RCIM_Cubic>
        ]}
      }}
     CurveIndex=&RandomRangeFloat(Minimum=0, Maximum=1)
    )
    Lifetime=&FloatFromCurve(
     FloatCurve=DataInterface() {propertyValues {Curve {keys [
         {arriveTangent=0.6956, interpMode=RCIM_Cubic, leaveTangent=0.6956, tangentMode=RCTM_User, time=0, value=0.1}
         {arriveTangentWeight=0.1692, interpMode=RCIM_Cubic, leaveTangentWeight=0.1692, tangentMode=RCTM_User, time=0.5, value=0.3}
         {arriveTangent=2.4814, tangentMode=RCTM_User, time=1, value=0.75}
        ]}}}
     CurveIndex=&RandomRangeFloat(Minimum=0, Maximum=1)
    )
    Position Offset:NiagaraPosition=$System.CenteredPos
    Position Offset Coordinate Space=::Local
    Sprite Size Mode=::NewEnumerator1
    Uniform Sprite Size Max=15
    UsePositionOffset=true
    Write Material Random=true
   )
   &AddVelocity(
    Rotation Coordinate Space=::Simulation
    Velocity=&RandomVector(Vector Scale=&Multiply_Float(
      A=&RandomRangeFloat(Minimum=125, Maximum=200)
      B=&MakeCustomFloatFromBool(Bool=$User.IsLargeBody, True Float=1.4, False Float=1.2)
     ))
   )
   &CameraOffset(Camera Offset Amount=-40)
   /SetVariables(Particles.Position=&AddVectorToPosition(
     Position=$Particles.Position
     Vector=&Multiply_VectorByFloat(
      Vector=&NormalizeVector(Vector To Normalize=$Particles.Velocity, Safe Normalize=true, Safe Fallback Threshold=0, Safe Fallback Vector=<0 0 0.001>)
      Float=10
     )
    )) {isSetParameters=true}
  ]
  particleUpdateScript [
   &ParticleState()
   &ScaleSpriteSize(
    Uniform Curve Scale=&MakeCustomFloatFromBool(Bool=$User.IsLargeBody, True Float=1.4, False Float=1)
    Uniform Curve Sprite Scale=DataInterface() {propertyValues {Curve {defaultValue=0, keys [{tangentMode=RCTM_User, time=0, value=1}, {tangentMode=RCTM_User, time=1, value=0}]}}}
   )
   &CurlNoiseForce(
    Bias Noise=false
    Noise Frequency=25
    Noise Strength=&FloatFromCurve(
     FloatCurve=DataInterface() {propertyValues {Curve {keys [{arriveTangent=-1.9055, interpMode=RCIM_Cubic, leaveTangent=-1.9055, tangentMode=RCTM_User, time=0, value=1}, <at 1 0.25>]}}}
     Scale Curve=500
    )
    Pan Noise=true
    Pan Noise Field=<0 0 1>
   )
   &Drag(Drag=2, Use Linear Drag=true, Use Rotational Drag=true)
   &SolveForcesAndVelocity(Clamp Velocity=false, Limit Acceleration=false)
   /SetVariables(
    Particles.BackerColor:LinearColor=<rgba 0.0176 0.0128 0.0885 0.5>
    Particles.BackerPivotOffset:Vector2f=&SelectVector2DFromArray(
     Array Sampling Mode=::NewEnumerator0
     Direct Array Index=?
     Vector 2D Selection Array=DataInterface() {propertyValues {FloatData [<0.45 0.45>, <0.45 0.55>, <0.55 0.45>, <0.55 0.55>]}}
     Random Seed=0
    )
    Particles.BackerScale:Vector2f=&Multiply_Vector2D_ByFloat(Float=1.75, Vector2D=$Particles.SpriteSize)
   ) {isSetParameters=true}
  ]
  rendererClasses [&NiagaraSpriteRendererProperties]
  renderers [
   {rendererClass=&NiagaraSpriteRendererProperties}
   {rendererClass=&NiagaraSpriteRendererProperties, rendererIndex=1}
  ]
 }
 Corona {
  emitterSpawnScript [/SetVariables() {isSetParameters=true}]
  emitterUpdateScript [
   &EmitterState()
   &SpawnBurst_Instantaneous(Spawn Count=1, Use Spawn Probability=false)
  ]
  particleSpawnScript [
   &InitializeParticle(
    Lifetime=0.25
    Position Offset:NiagaraPosition=$System.CenteredPos
    Sprite Rotation Mode=::NewEnumerator2
    Sprite Size Mode=::NewEnumerator3
    Uniform Sprite Size=400
    UsePositionOffset=true
    Write Material Random=true
   )
   &DynamicMaterialParameters(
    Index 0 Param 1=&RandomRangeFloat(Minimum=0.4, Maximum=0.6)
    Index 0 Param 2=1
    Index 0 Param 3=1
    Index 0 Param 4=1
    Module.Param0WriteEnabled=true
    Module.Param1WriteEnabled=true
    Module.Param2WriteEnabled=true
    Module.Param3WriteEnabled=true
   )
  ]
  particleUpdateScript [
   &ParticleState()
   &ScaleColor(
    Scale Alpha=&Multiply_Float(
     A=&FNTimeOfDayScalar(Night Scale=1, Morning Scale=1, Evening Scale=1)
     B=&FloatFromCurve(FloatCurve=DataInterface() {propertyValues {Curve {keys [
          {arriveTangent=9.6062, interpMode=RCIM_Cubic, leaveTangent=9.6062, tangentMode=RCTM_User, time=0, value=0}
          {interpMode=RCIM_Cubic, tangentMode=RCTM_User, time=0.2, value=1}
          {interpMode=RCIM_Cubic, tangentMode=RCTM_User, time=1, value=0}
         ]}}})
    )
    Scale RGB=&Multiply_VectorByFloat(Vector=<1 1 1>, Float=1)
    ScaleA=true
    ScaleRGB=true
   )
   &ScaleSpriteSize(Scale Factor=&Vector2DFromFloat(Value=&FloatFromCurve(FloatCurve=DataInterface() {propertyValues {Curve {keys [
          {arriveTangent=2.1722, interpMode=RCIM_Cubic, leaveTangent=2.1722, tangentMode=RCTM_User, time=0, value=0}
          {arriveTangent=0.0224, interpMode=RCIM_Cubic, leaveTangent=0.0224, tangentMode=RCTM_User, time=1, value=1}
         ]}}})))
   &CameraOffset(Camera Offset Amount=-30)
   &DynamicMaterialParameters(
    Index 0 Param 1=0.75
    Index 0 Param 2=2
    Index 0 Param 3=1
    Module.Param0WriteEnabled=true
    Module.Param1WriteEnabled=true
    Module.Param2WriteEnabled=true
    Module.Param3WriteEnabled=true
   ) {module=DynamicMaterialParameters001}
   /SetVariables(Particles.Position=$System.CenteredPos) {isSetParameters=true}
  ]
  rendererClasses [&NiagaraSpriteRendererProperties]
  renderers [{rendererClass=&NiagaraSpriteRendererProperties}]
 }
 DarkGlow {
  emitterUpdateScript [
   &EmitterState()
   &SpawnBurst_Instantaneous(Spawn Count=1, Use Spawn Probability=false)
  ]
  particleSpawnScript [
   &InitializeParticle(
    Color=<rgba 0.1249 0.0652 0.221 1>
    Lifetime=0.4
    Position Offset:NiagaraPosition=$System.CenteredPos
    Sprite Rotation Mode=::NewEnumerator2
    Sprite Size Mode=::NewEnumerator3
    Uniform Sprite Size=230
    UsePositionOffset=true
    Write Material Random=true
   )
   &CameraOffset(Camera Offset Amount=-50)
  ]
  particleUpdateScript [
   &ParticleState()
   &ScaleColor(
    Scale Alpha=&FloatFromCurve(
     FloatCurve=DataInterface() {propertyValues {Curve {keys [
         {interpMode=RCIM_Cubic, tangentMode=RCTM_User, time=0, value=0}
         {interpMode=RCIM_Cubic, tangentMode=RCTM_User, time=0.2, value=1}
         {interpMode=RCIM_Cubic, tangentMode=RCTM_User, time=0.5, value=1}
         {tangentMode=RCTM_User, time=1, value=0}
        ]}}}
     Scale Curve=2
    )
    ScaleA=true
    ScaleRGB=true
   )
   /SetVariables(Particles.Position=$System.CenteredPos) {isSetParameters=true}
  ]
  rendererClasses [&NiagaraSpriteRendererProperties]
  renderers [{rendererClass=&NiagaraSpriteRendererProperties}]
 }
 Flare {
  emitterUpdateScript [
   &EmitterState()
   &SpawnBurst_Instantaneous(Spawn Count=1, Use Spawn Probability=false)
  ]
  particleSpawnScript [
   &InitializeParticle(
    Color=<rgba 0 1.25 0.9792 1>
    Lifetime=0.3
    Position Offset:NiagaraPosition=$System.CenteredPos
    Sprite Rotation Mode=::NewEnumerator2
    Sprite Size Mode=::NewEnumerator3
    Uniform Sprite Size=600
    UsePositionOffset=true
    Write Material Random=true
   )
   &DynamicMaterialParameters(
    Index 0 Param 1=&RandomRangeFloat(Minimum=0.4, Maximum=0.6)
    Index 0 Param 2=1
    Index 0 Param 3=1
    Index 0 Param 4=1
    Module.Param0WriteEnabled=true
    Module.Param1WriteEnabled=true
    Module.Param2WriteEnabled=true
    Module.Param3WriteEnabled=true
   )
   /SetVariables(Particles.Sharp Flare Camera Offset=40) {isSetParameters=true}
   &CameraOffset(Camera Offset Amount=20) {module=CameraOffset001}
  ]
  particleUpdateScript [
   &ParticleState()
   &ScaleColor(
    Scale Alpha=&FloatFromCurve(FloatCurve=DataInterface() {propertyValues {Curve {keys [<at 0 0>,<at 0.2 1>,<at 0.8 1>,<at 1 0>]}}})
    Scale RGB=&MakeVectorFromLinearColorRGB(Color=&ColorFromCurve(ColorCurve=DataInterface() {propertyValues {
        AlphaCurve {keys [<at 0 1 RCIM_Cubic>]}
        BlueCurve {keys [<at 0 1 RCIM_Cubic>,<at 1 1 RCIM_Cubic>]}
        GreenCurve {keys [<at 0 1 RCIM_Cubic>,<at 1 0.1462 RCIM_Cubic>]}
        RedCurve {keys [<at 0 1 RCIM_Cubic>,<at 1 0 RCIM_Cubic>]}
       }}))
    ScaleA=true
    ScaleRGB=true
   )
   &ScaleSpriteSize(Scale Factor=&Vector2DFromFloat(Value=&FloatFromCurve(FloatCurve=DataInterface() {propertyValues {Curve {keys [<at 0 1.5 RCIM_Cubic>,<at 1 0>]}}})))
   /SetVariables(Particles.Flash color:LinearColor=&MakeLinearColorFromVectorAndFloat(
     "Vector (RGB)"=&MakeVectorFromLinearColorRGB(Color=<rgba 0.5235 0.3277 1 1>)
     "Float (Alpha)"=&Multiply_Float(
      A=&Multiply_Float(
       A=&FNTimeOfDayScalar(Night Scale=1, Morning Scale=1, Evening Scale=1)
       B=&FloatFromCurve(FloatCurve=DataInterface() {propertyValues {Curve {keys [<at 0 0>,<at 0.124 1>,<at 0.496 1>,<at 0.62 0>]}}})
      )
      B=1
     )
    )) {isSetParameters=true}
   /SetVariables(Particles.Flash size:Vector2f=&Multiply_Vector2D_ByFloat(
     Float=&Multiply_Float(
      A=400
      B=&FloatFromCurve(FloatCurve=DataInterface() {propertyValues {Curve {keys [
           {arriveTangent=4.1785, interpMode=RCIM_Cubic, leaveTangent=4.1785, tangentMode=RCTM_User, time=0, value=0}
           {arriveTangent=-0.15, interpMode=RCIM_Cubic, leaveTangent=-0.15, tangentMode=RCTM_User, time=0.1378, value=1}
           {arriveTangent=-0.0405, interpMode=RCIM_Cubic, leaveTangent=-0.0405, tangentMode=RCTM_User, time=0.62, value=0.25}
          ]}}})
     )
     Vector2D=<1 1>
    )) {isSetParameters=true}
   /SetVariables(Particles.Position=$System.CenteredPos) {isSetParameters=true}
  ]
  rendererClasses [&NiagaraSpriteRendererProperties]
  renderers [
   {rendererClass=&NiagaraSpriteRendererProperties}
   {rendererClass=&NiagaraSpriteRendererProperties, rendererIndex=1}
  ]
 }
 Rays {
  emitterUpdateScript [
   &EmitterState()
   &SpawnBurst_Instantaneous(Spawn Count=10, Use Spawn Probability=false, UseLoopCountLimit=true)
  ]
  particleSpawnScript [
   &InitializeParticle(
    Color=&ColorFromCurve(
     ColorCurve=DataInterface() {propertyValues {
       AlphaCurve {keys [<at 0 1 RCIM_Cubic>]}
       BlueCurve {keys [
         <at 0 0.7157 RCIM_Cubic>
         {arriveTangentWeight=0.0579, interpMode=RCIM_Cubic, leaveTangentWeight=0.1039, tangentMode=RCTM_User, time=0.1568, value=2}
         {arriveTangent=-6.0034, interpMode=RCIM_Cubic, leaveTangent=-6.0034, time=0.3653, value=0.5068}
         {arriveTangent=-2.0605, interpMode=RCIM_Cubic, leaveTangent=-2.0605, time=0.4785, value=0.0689}
         {arriveTangent=-0.2853, interpMode=RCIM_Cubic, leaveTangent=-0.2853, time=0.6113, value=0}
         {arriveTangent=0.0766, interpMode=RCIM_Cubic, leaveTangent=0.0766, time=0.7199, value=0}
         {arriveTangent=2.5554, interpMode=RCIM_Cubic, leaveTangent=2.5554, time=0.8722, value=0.02}
         <at 1 0.7157 RCIM_Cubic>
        ]}
       GreenCurve {keys [
         <at 0 0.4969 RCIM_Cubic>
         {arriveTangent=-1.3602, interpMode=RCIM_Cubic, leaveTangent=-1.3602, time=0.1568, value=0.1369}
         {arriveTangent=-0.2799, interpMode=RCIM_Cubic, leaveTangent=-0.2799, time=0.3653, value=0}
         {arriveTangent=1.0968, interpMode=RCIM_Cubic, leaveTangent=1.0968, time=0.4785, value=0.0469}
         {arriveTangent=3.7429, interpMode=RCIM_Cubic, leaveTangent=3.7429, time=0.6113, value=0.2698}
         {arriveTangent=2.7988, interpMode=RCIM_Cubic, leaveTangent=2.7988, time=0.7199, value=0.9505}
         {arriveTangent=-1.6195, interpMode=RCIM_Cubic, leaveTangent=-1.6195, time=0.8722, value=1}
         <at 1 0.4969 RCIM_Cubic>
        ]}
       RedCurve {keys [
         <at 0 0 RCIM_Cubic>
         {arriveTangentWeight=0.1864, interpMode=RCIM_Cubic, leaveTangentWeight=0.3345, tangentMode=RCTM_User, time=0.1568, value=0}
         {arriveTangentWeight=0.2264, interpMode=RCIM_Cubic, leaveTangentWeight=0.1399, tangentMode=RCTM_User, time=0.3653, value=1.5}
         {arriveTangentWeight=0.1082, interpMode=RCIM_Cubic, leaveTangentWeight=0.0892, tangentMode=RCTM_User, time=0.4785, value=1}
         {arriveTangentWeight=0.0478, interpMode=RCIM_Cubic, leaveTangentWeight=0.0203, tangentMode=RCTM_User, time=0.6113, value=1}
         {arriveTangentWeight=0.175, interpMode=RCIM_Cubic, leaveTangentWeight=0.16, tangentMode=RCTM_User, time=0.7199, value=1}
         {arriveTangentWeight=0.1029, interpMode=RCIM_Cubic, leaveTangentWeight=0.236, tangentMode=RCTM_User, time=0.8722, value=0.0018}
         <at 1 0 RCIM_Cubic>
        ]}
      }}
     CurveIndex=&RandomRangeFloat(Minimum=0, Maximum=1)
    )
    Lifetime Max=1
    Lifetime Mode=::NewEnumerator1
    Position Offset:NiagaraPosition=$System.CenteredPos
    Sprite Rotation Angle=&ReturnNormalizedExecIndex(Normalized Index Includes One=false)
    Sprite Rotation Mode=::NewEnumerator1
    Sprite Size Max=<100 100>
    Sprite Size Min=<50 50>
    Sprite Size Mode=::NewEnumerator2
    UsePositionOffset=true
    Write Material Random=true
   )
   &CameraOffset(Camera Offset Amount=-10)
  ]
  particleUpdateScript [
   &ParticleState()
   &ScaleSpriteSize(
    Non-Uniform Curve Sprite Scale=DataInterface() {propertyValues {
      XCurve {keys [{tangentMode=RCTM_User, time=0, value=1}, {tangentMode=RCTM_User, time=1, value=0.5}]}
      YCurve {keys [
        {arriveTangent=15.8763, interpMode=RCIM_Cubic, leaveTangent=15.8763, tangentMode=RCTM_User, time=0, value=0}
        {
         arriveTangent=-0.3643
         arriveTangentWeight=0.1414
         interpMode=RCIM_Cubic
         leaveTangent=-0.3643
         leaveTangentWeight=0.33
         tangentMode=RCTM_User
         time=0.4
         value=1
        }
        {arriveTangent=-0.1853, tangentMode=RCTM_User, time=1, value=0.7775}
       ]}
     }}
    Scale Sprite Size Mode=::NewEnumerator3
   )
   &SpriteRotationRate(Rotation Rate=&RandomRangeFloat(Minimum=30, Maximum=75))
   &ScaleColor(
    Scale Alpha=&FloatFromCurve(FloatCurve=DataInterface() {propertyValues {Curve {keys [{arriveTangent=-2.6228, interpMode=RCIM_Cubic, leaveTangent=-2.6228, tangentMode=RCTM_User, time=0, value=1}, <at 1 0>]}}})
    Scale RGB=&MakeVectorFromLinearColorRGB(Color=$Particles.Initial.Color)
    ScaleA=true
    ScaleRGB=true
   )
   /SetVariables(Particles.Position=$System.CenteredPos) {isSetParameters=true}
  ]
  rendererClasses [&NiagaraSpriteRendererProperties]
  renderers [{rendererClass=&NiagaraSpriteRendererProperties}]
 }
}
summary {
 systemName=NS_Aura_Kaleido_Start
 emitters [
  {emitterName=Flare, rendererClasses [&NiagaraSpriteRendererProperties]}
  {emitterName=DarkGlow, rendererClasses [&NiagaraSpriteRendererProperties]}
  {emitterName=Corona, rendererClasses [&NiagaraSpriteRendererProperties]}
  {emitterName=Rays, rendererClasses [&NiagaraSpriteRendererProperties]}
  {emitterName=CirclesOut, simTarget=GPUComputeSim, rendererClasses [&NiagaraSpriteRendererProperties]}
 ]
 userVariables [User.IsLargeBody {type=NiagaraBool}]
}
